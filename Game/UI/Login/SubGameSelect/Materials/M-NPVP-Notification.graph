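# Translucent UI material for a new-PVP notification image on the sub-game select screen of the login UI.
# It samples a texture parameter, defaulting to the new-PVP texture, using the non-tiling 9-slice UVs from GetUserInterfaceUV.
# A scalar parameter lerps those UVs toward a copy with inverted U, so the image can be mirrored horizontally.
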Material /Game/UI/Login/SubGameSelect/Materials/M-NPVP-Notification
graph {
 properties {
  bUsedWithStaticMesh=false
  blendMode=BLEND_Translucent
  materialDomain=MD_UI
  pixelDepthOffsetMode=PDOM_Legacy
  usedShadingModels=Default Lit
 }
 nodes [
  n1 = MaterialFunctionCall() {materialFunction=&GetUserInterfaceUV}
  n2 = OneMinus(None=ComponentMask(None=@n1."9-Slice UV (No Tiling)") {r=true})
  n3 = ComponentMask(None=@n1."9-Slice UV (No Tiling)") {g=true}
  TextureSampleParameter2D(UVs=LinearInterpolate(A=@n1."9-Slice UV (No Tiling)", B=AppendVector(A=@n2, B=@n3), Alpha=ScalarParameter() {parameterName=sFlipUChannel, sliderMax=1})) {parameterName=Texture, texture=/Game/UI/Login/SubGameSelect/Textures/T_newPVP_Texture}
 ]
}
