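# Translucent UI material for a circular focus-border pulse in sub-game select, with an optional animated arrow.
# Time, or an override time, loops through a power curve to expand, thin and fade a ring cut from a radial gradient by two smoothsteps.
# A static switch adds a chevron made by subtracting scaled, rotated UV square masks and cropping vertically; its thickness and sine-eased offset follow the same pulse.
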Material /Game/UI/Login/SubGameSelect/Materials/M_UI_FocusBorderPulse_Circle
graph {
 properties {
  bUsedWithStaticMesh=false
  blendMode=BLEND_Translucent
  materialDomain=MD_UI
  pixelDepthOffsetMode=PDOM_Legacy
  usedShadingModels=Default Lit
 }
 nodes [
  Constant() {r=1}
  n1 = ScalarParameter() {defaultValue=0.01, parameterName=EdgeSoftness}
  n2 = Multiply(A=Time(), B=ScalarParameter() {defaultValue=1, parameterName=PulseSpeed})
  n3 = ScalarParameter() {parameterName=OverrideTime}
  n4 = ScalarParameter() {parameterName=UseSetTime}
  n5 = Power(
   Base=Fmod(A=LinearInterpolate(A=@n2, B=@n3, Alpha=@n4), B=Constant() {r=1})
   Exp=ScalarParameter() {defaultValue=2, parameterName=FalloffPower}
  )
  n6 = LinearInterpolate(A=ScalarParameter() {defaultValue=1, parameterName=OuterStartScale}, B=ScalarParameter() {parameterName=OuterEndScale}, Alpha=@n5)
  n7 = MaterialFunctionCall(
   UVs=MaterialFunctionCall() {materialFunction=&GetUserInterfaceUV}.9-Slice UV
   Radius=ScalarParameter() {defaultValue=0.5, parameterName=DefaultRadius}
   Density=ScalarParameter() {defaultValue=2.33, parameterName=DefaultDensity}
  ) {materialFunction=&RadialGradientExponential}
  n8 = Clamp(None=MaterialFunctionCall(Alpha=@n7.RadialGradientExponential, Min=@n6, Max=Add(A=@n6, B=@n1)) {materialFunction=&SmoothStep}.Result)
  n9 = Add(
   A=@n6
   B=LinearInterpolate(
    A=ScalarParameter() {defaultValue=0.1, parameterName=InnerStartScale}
    B=ScalarParameter() {defaultValue=0.01, parameterName=InnerEndScale}
    Alpha=@n5
   )
  )
  n10 = Add(A=@n9, B=ScalarParameter() {defaultValue=0.01, parameterName=EdgeSoftness})
  n11 = OneMinus(None=Clamp(None=MaterialFunctionCall(Alpha=@n7.RadialGradientExponential, Min=@n9, Max=@n10) {materialFunction=&SmoothStep}.Result))
  n12 = ScalarParameter() {defaultValue=1, parameterName=StartOpacity}
  n13 = ScalarParameter() {parameterName=EndOpacity}
  n14 = ScalarParameter() {group=Arrow, parameterName=ArrowPosEnd}
  n15 = ScalarParameter() {group=Arrow, parameterName=ArrowPosStart}
  n16 = Sine(None=OneMinus(None=@n5))
  n17 = Add(A=TextureCoordinate(), B=AppendVector(A=Constant(), B=LinearInterpolate(A=@n14, B=@n15, Alpha=@n16)))
  n18 = MaterialFunctionCall(UVs=@n17, "Rotation Angle (0-1)"=ScalarParameter() {defaultValue=0.125, group=Arrow, parameterName=ArrowAngle}) {materialFunction=&CustomRotator}
  n19 = ScalarParameter() {defaultValue=0.7, group=Arrow, parameterName=ArrowScale}
  n20 = LinearInterpolate(
   A=ScalarParameter() {defaultValue=0.1, group=Arrow, parameterName=ArrowStartScale}
   B=ScalarParameter() {defaultValue=0.01, group=Arrow, parameterName=ArrowEndScale}
   Alpha=@n5
  )
  n21 = Subtract(
   A=MaterialFunctionCall(UVs=@n18.Rotated Values, Texture Scale=@n19) {materialFunction=&ScaleUVsByCenter}.0-1 mask
   B=MaterialFunctionCall(UVs=@n18.Rotated Values, Texture Scale=Subtract(A=@n19, B=@n20)) {materialFunction=&ScaleUVsByCenter}.0-1 mask
  )
  n22 = ScalarParameter() {defaultValue=0.85, group=Arrow, parameterName=ArrowCutoff}
  n23 = MaterialFunctionCall(Alpha=ComponentMask(None=@n17) {g=true}, Min=@n22, Max=Add(A=@n22, B=ScalarParameter() {defaultValue=0.01, parameterName=EdgeSoftness})) {materialFunction=&SmoothStep}
  Max(
   A=Multiply(A=Multiply(A=@n8, B=@n11), B=LinearInterpolate(A=@n12, B=@n13, Alpha=@n5) {constA=1, constB=0})
   B=StaticSwitchParameter(True=Multiply(A=@n21, B=@n23.Result), False=Constant()) {group=Arrow, parameterName=ShowArrow}
  )
 ]
}
