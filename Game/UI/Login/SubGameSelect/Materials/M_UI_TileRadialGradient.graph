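# Translucent UI material in the login SubGameSelect folder that computes a radial gradient, by its name for tiles.
# Normalized UI UVs are shifted by 0.5, scaled by the Size parameter, then fed into RadialGradientExponential with Center, Radius and Density parameters.
# Size and Center each use only their red channel for both axes, so scaling is uniform and the center moves only diagonally.
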Material /Game/UI/Login/SubGameSelect/Materials/M_UI_TileRadialGradient
graph {
 properties {
  bUsedWithStaticMesh=false
  blendMode=BLEND_Translucent
  materialDomain=MD_UI
  pixelDepthOffsetMode=PDOM_Legacy
  usedShadingModels=Default Lit
 }
 nodes [
  Constant() {r=1}
  n1 = MaterialFunctionCall() {materialFunction=&GetUserInterfaceUV}
  n2 = VectorParameter() {parameterName=Size, defaultValue=<rgba 2 2 0 1>}
  n3 = VectorParameter() {parameterName=Center, defaultValue=<rgba 0 1 0 1>}
  MaterialFunctionCall(
   UVs=Multiply(A=Subtract(A=@n1.Normalized UV) {constB=0.5}, B=AppendVector(A=@n2.R, B=@n2.R)) {constB=2}
   CenterPosition=AppendVector(A=@n3.R, B=@n3.R)
   Radius=ScalarParameter() {defaultValue=1.2, parameterName=Radius}
   Density=ScalarParameter() {defaultValue=1, parameterName=Density}
  ) {materialFunction=&RadialGradientExponential}
 ]
}
